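# Translucent UI material for the Rocket Drill fuel gauge, a radial meter that shows fuel level and boost state.
# Progress drives a radial arc fill and shifts its color between critical and normal, with onBoosting selecting the boost color and a sine-pulsed pip background.
# SDF shapes build the frame and ring, while channel-packed distance fields from a temporary texture supply inner and frame masks.
# Named reroutes link these parts.
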Material /RocketDrillGameplay/Assets/UI/M_UI_RocketDrill_Fuel
graph {
 properties {bAutomaticallySetUsageInEditor=false, blendMode=BLEND_Translucent, materialDomain=MD_UI, usedShadingModels=Default Lit}
 nodes [
  n1 = NamedRerouteUsage()
  n2 = ScalarParameter() {defaultValue=0.85, parameterName=BarOpacity}
  n3 = NamedRerouteUsage()
  n4 = ConstantBiasScale(None=Sine(None=Time()) {period=0.5})
  n5 = LinearInterpolate(A=@n3, B=Multiply(A=@n3, B=LinearInterpolate(Alpha=@n4) {constA=0.4}), Alpha=ScalarParameter() {parameterName=onBoosting})
  Multiply(
   A=Max(A=Multiply(A=@n1, B=@n2), B=NamedRerouteDeclaration(None=@n5) {name=PipBg, nodeColor=<rgba 0 1 0.4141 1>})
   B=ScalarParameter() {defaultValue=0.85, parameterName=MasterOpacity}
  )
  n6 = MaterialFunctionCall!(UVs, Radius=Constant() {r=0.7}) {materialFunction=/FortUILibrary/MaterialFunctions/Shapes/SDFs/MF_SDF_Circle}
  n7 = NamedRerouteUsage()
  n8 = TextureCoordinate()
  n9 = Abs(None=ConstantBiasScale(None=@n8) {bias=-0.5, scale=2})
  n10 = Min(A=ComponentMask(None=@n9) {g=true}, B=ComponentMask(None=@n9) {r=true})
  n11 = MaterialFunctionCall!(UVs=Rotator(Coordinate=@n8, Time=Constant() {r=0.5}) {speed=3.1416}, Sides=Constant() {r=4}) {materialFunction=/FortUILibrary/MaterialFunctions/Basics/MF_UI_UVRadialMirror}
  n12 = Constant() {r=0.1}
  n13 = Constant() {r=-0.075}
  n14 = ComponentMask(None=Add(A=Rotator(Coordinate=@n11.UVs, Time=@n12) {speed=3.1416}, B=@n13) {constB=-0.1}) {r=true}
  n15 = Constant() {r=0.9}
  n16 = ComponentMask(None=Add(A=Rotator(Coordinate=@n11.UVs, Time=@n15) {speed=3.1416}, B=@n13) {constB=-0.1}) {r=true}
  n17 = Add(A=Rotator(Coordinate=@n8, Time=Constant() {r=0.25}) {speed=3.1416}) {constB=-0.65}
  n18 = Max(A=Max(A=Min(A=@n7, B=@n10), B=Min(A=@n14, B=@n16)), B=Multiply(A=ComponentMask(None=@n17) {g=true}) {constB=-2})
  n19 = MaterialFunctionCall!(UVs=ConstantBiasScale(None=@n8) {bias=-0.5, scale=1}, Size=Constant() {r=1}, CornerRoundness, KeepAspectRatio, PixelSize) {materialFunction=/FortUILibrary/MaterialFunctions/Shapes/SDFs/MF_SDF_Box}
  n20 = Max(
   A=MaterialFunctionCall!(DistanceField=@n6.SDF, Threshold, Thickness=Constant() {r=0.03}, Alignment, Smoothing) {materialFunction=/FortUILibrary/MaterialFunctions/Shapes/Modifiers/MF_UI_ShapeOutline}.SDFOutline
   B=@n19.SDF
  )
  n21 = Multiply(A=Max(A=@n6.SDF, B=Min(A=@n18, B=@n20))) {constB=3}
  NamedRerouteDeclaration(None=MaterialFunctionCall(DistanceField=NamedRerouteDeclaration(None=@n21) {name=OuterFrameSDF, nodeColor=<rgba 0 0 0 1>}, Threshold=Constant() {r=0.01}) {materialFunction=&MF_UI_AntiAliasing}.Inner) {name=OuterFrame, nodeColor=<rgba 1 0.9844 0 1>}
  NamedRerouteDeclaration(None=MaterialFunctionCall() {materialFunction=&GetUserInterfaceUV}.Normalized UV) {name=UV, nodeColor=<rgba 0 0 0 1>}
  n22 = MaterialFunctionCall(
   Size=Constant() {r=0.62}
   OutlineThickness=ScalarParameter() {defaultValue=0.015, group=AmmoCounter, parameterName=RadialThickness, sliderMax=1}
  ) {materialFunction=&MF_UI_Circle}
  NamedRerouteDeclaration(None=OneMinus(None=SmoothStep(Value=@n22.SDFOutline) {constMax=0.01})) {name=Mask_Radial, nodeColor=<rgba 0 0 0 1>}
  n23 = MaterialFunctionCall!(
   UV=MaterialFunctionCall(UVs=NamedRerouteUsage(), "Rotation Angle (0-1)"=Constant() {r=0.5}) {materialFunction=&CustomRotator}.Rotated Values
   "Arc Start Angle (0-1)"=ScalarParameter() {defaultValue=0.99, parameterName=Start Angle, sliderMax=1, sortPriority=31}
   "Arc End Angle (0-1)"=ScalarParameter() {defaultValue=0.24, parameterName=End Angle, sliderMax=1}
   "Fill Ratio (0-1)"=NamedRerouteUsage()
   Fill Clockwise
  ) {materialFunction=/FortUILibrary/MaterialFunctions/Basics/MF_UI_RadialArc}
  n24 = SmoothStep(Value=@n23.Radial Fill) {constMax=0.001}
  NamedRerouteDeclaration(None=Multiply(A=NamedRerouteUsage(), B=@n24)) {name=Mask_OverheatProgress, nodeColor=<rgba 0 0 0 1>}
  n25 = TextureCoordinate()
  n26 = Constant2Vector() {g=-0.12, r=-0.146}
  NamedRerouteDeclaration(None=MaterialFunctionCall(UVs=Add(A=@n25, B=@n26), Texture Scale=Constant() {r=0.36}) {materialFunction=&ScaleUVsByCenter}.UVs) {name=TextureUVs, nodeColor=<rgba 0 0 0 1>}
  n27 = TextureSample(UVs=NamedRerouteUsage()) {texture=&/TEMP_T_UI_RocketDrill_Fuel}
  n28 = NamedRerouteDeclaration(None=MaterialFunctionCall!(RGBA Channels=@n27.B, Glow Min=Constant() {r=0.1}, Glow Max=Constant() {r=0.8}, Outline Thickness) {materialFunction=/FortUILibrary/MaterialFunctions/Basics/MF_UI_ChannelPackedDistanceField}.Fill) {name=InnerMask, nodeColor=<rgba 0 0 0 1>}
  n29 = NamedRerouteUsage()
  NamedRerouteDeclaration(None=Saturate(None=Multiply(A=@n28, B=@n29))) {name=Overheat Reached Glow, nodeColor=<rgba 0 0 0 1>}
  NamedRerouteDeclaration(None=Multiply(A=@n19.SDF) {constB=-1}) {name=BLQuad2, nodeColor=<rgba 0 0 0 1>}
  n30 = Constant3Vector()
  n31 = VectorParameter() {parameterName=FrameColor, defaultValue=<rgba 1 1 1 1>}
  n32 = NamedRerouteUsage()
  NamedRerouteDeclaration(None=LinearInterpolate(A=NamedRerouteUsage(), B=LinearInterpolate(A=@n30, B=@n31.RGB, Alpha=@n32), Alpha=@n32)) {name=Outer Frame Color, nodeColor=<rgba 0 0 0 1>}
  n33 = VectorParameter() {group=Color, parameterName=DefaultColor, defaultValue=<rgba 1 1 1 1>}
  n34 = VectorParameter() {group=Color, parameterName=BoostingColor, defaultValue=<rgba 1 0.6724 0.0578 1>}
  n35 = ScalarParameter() {parameterName=onBoosting}
  n36 = ScalarParameter() {group=Color, parameterName=Progress}
  n37 = ScalarParameter() {defaultValue=0.15, group=Color, parameterName=ColorBEndProgress}
  n38 = ScalarParameter() {defaultValue=0.15, group=Color, parameterName=ColorCStartProgress}
  NamedRerouteDeclaration(None=LinearInterpolate(
    A=VectorParameter() {group=Color, parameterName=CriticalColor, defaultValue=<rgba 1 0.0003 0 1>}.RGB
    B=LinearInterpolate(A=@n33.RGB, B=@n34.RGB, Alpha=@n35)
    Alpha=MaterialFunctionCall!(Input=@n36, Input Low=@n37, Input High=@n38) {materialFunction=/Engine/Functions/Engine_MaterialFunctions03/Math/RemapValueRangeNormalized}.Result
   ) {constA=1}) {name=Overheat Color, nodeColor=<rgba 0 0 0 1>}
  n39 = MaterialFunctionCall(
   UV=MaterialFunctionCall(UVs=NamedRerouteUsage(), "Rotation Angle (0-1)"=Constant() {r=0.5}) {materialFunction=&CustomRotator}.Rotated Values
   "Arc Start Angle (0-1)"=Constant() {r=0.97}
   "Arc End Angle (0-1)"=Constant() {r=0.25}
   "Fill Ratio (0-1)"=NamedRerouteUsage()
  ) {materialFunction=&MF_UI_RadialArc}
  NamedRerouteDeclaration(None=Saturate(None=Multiply(A=@n39.Radial Fill) {constB=50})) {name=Progress Radial Fill, nodeColor=<rgba 0 0 0 1>}
  n40 = MaterialFunctionCall!("Value (0-1)"=OneMinus(None=@n23.Arc Mask), Steps) {materialFunction=/FortUILibrary/MaterialFunctions/Basics/MF_UI_StepRatio}
  n41 = Multiply(A=Saturate(None=Add(A=@n40.Stepped Gradient) {constB=0.15})) {constB=0.8}
  NamedRerouteDeclaration(None=Multiply(A=Max(A=@n24, B=@n41)) {constB=0.8}) {name=OverheatProgress, nodeColor=<rgba 0 0 0 1>}
  NamedRerouteDeclaration(None=OneMinus(None=ScalarParameter() {group=Color, parameterName=Progress, sliderMax=1})) {name=Progress, nodeColor=<rgba 0 0 0 1>}
  n42 = MaterialFunctionCall(RGBA Channels=@n27.R, Glow Min=Constant() {r=0.25}, Glow Max=Constant() {r=0.5}) {materialFunction=&MF_UI_ChannelPackedDistanceField}
  NamedRerouteDeclaration(None=Add(A=SmoothStep(Value=@n27.R) {constMin=0.2}, B=NamedRerouteDeclaration(None=@n42.Fill) {name=OuterFrame_Mask, nodeColor=<rgba 0 0 0 1>})) {name=OuterFrame_Shadow, nodeColor=<rgba 0 0 0 1>}
  NamedRerouteUsage()
 ]
}
