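# Translucent UI material that draws the Rocket Drill riding reticle procedurally, blending between mirrored chevrons and a small circle through CircleOrChevron.
# Each chevron is the max of two fixed-angle rotated UV gradients on symmetry-mirrored coordinates positioned by ArrowOffset and ArrowScale, while the circle uses an SDF function.
# Anti-aliased inner masks form the solid shape, and wider thresholds blend into a second mask that adds a border, scaled by MasterOpacity.
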Material /RocketDrillGameplay/Assets/UI/M_UI_RocketDrill_Riding_Reticle
graph {
 properties {bAutomaticallySetUsageInEditor=false, blendMode=BLEND_Translucent, materialDomain=MD_UI, usedShadingModels=Default Lit}
 nodes [
  n1 = ScalarParameter() {defaultValue=1, parameterName=ArrowOffset}
  n2 = ScalarParameter() {defaultValue=1, parameterName=ArrowScale}
  n3 = MaterialFunctionCall!(
   UV=TextureCoordinate()
   Offset=AppendVector(A=Constant(), B=LinearInterpolate(Alpha=@n1) {constA=-1.1, constB=-0.3})
   Rotation
   Scale=AppendVector(A=LinearInterpolate(Alpha=@n2) {constA=0.125, constB=0.2}, B=Constant() {r=0.2})
   MirrorY
  ) {materialFunction=/FortUILibrary/MaterialFunctions/Basics/MF_UI_Symmetry_2x1_DEPRECATED}
  n4 = Constant() {r=-0.4}
  n5 = Multiply(A=@n3.Result) {constB=-1}
  n6 = Add(A=@n4) {constB=0.075}
  n7 = Max(
   A=ComponentMask(None=Rotator(Coordinate=@n3.Result, Time=@n4) {speed=3.1416}) {r=true}
   B=ComponentMask(None=Rotator(Coordinate=@n5, Time=@n6) {speed=3.1416}) {r=true}
  )
  n8 = MaterialFunctionCall(DistanceField=@n7) {materialFunction=&MF_UI_AntiAliasing}
  n9 = MaterialFunctionCall!(UVs, Radius=Constant() {r=0.075}) {materialFunction=/FortUILibrary/MaterialFunctions/Shapes/SDFs/MF_SDF_Circle}
  n10 = MaterialFunctionCall(DistanceField=@n9.SDF) {materialFunction=&MF_UI_AntiAliasing}
  n11 = ScalarParameter() {defaultValue=1, parameterName=CircleOrChevron}
  LinearInterpolate(A=@n8.Inner, B=@n10.Inner, Alpha=@n11)
  n12 = MaterialFunctionCall(DistanceField=@n7, Threshold=Constant() {r=0.07}) {materialFunction=&MF_UI_AntiAliasing}
  n13 = MaterialFunctionCall(DistanceField=@n9.SDF, Threshold=Constant() {r=0.02}) {materialFunction=&MF_UI_AntiAliasing}
  Multiply(
   A=LinearInterpolate(
    A=LinearInterpolate(A=@n8.Inner, B=@n12.Inner) {constAlpha=0.6}
    B=LinearInterpolate(A=@n10.Inner, B=@n13.Inner) {constAlpha=0.4}
    Alpha=@n11
   )
   B=LinearInterpolate(Alpha=ScalarParameter() {defaultValue=1, parameterName=MasterOpacity}) {constB=0.85}
  )
 ]
}
