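# Reticle extension widget for the Rocket Drill weapon, layering center, frame and fuel material images in an overlay.
# On construct it caches the equipped ranged weapon and zeroes opacity on the bottom-frame and fuel dynamic materials.
# Async listeners for the pawn's RocketDrill.Active tag play OnRiding forward or reversed to animate those material parameters.
# The fuel image stays Hidden, OnBoosing and OnCooldown are never played here, and UpdateAmmo is empty.
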WidgetBlueprint /RocketDrillGameplay/Assets/UI/WBP_RocketDrill_Reticle
animations [
 OnBoosing {
  binding_count=2
  length_seconds=0.3
  object=&self:OnBoosing
  sequence {
   bindings [
    Image_ReticleCenter {
     id=#1
     possessed_class=Image
     tracks [{
       type=MovieSceneWidgetMaterialTrack
       sections [{
         type=MovieSceneParameterSection
         end_frame=9
         start_frame=0
         channels [ArrowOffset {type=MovieSceneScriptingFloatChannel, keys [[2,1],[5,1.3],[9,1]]}]
        }]
      }]
    }
    Image_Reticle_Bottom_Frame {
     id=#2
     possessed_class=Image
     tracks [{
       type=MovieSceneWidgetMaterialTrack
       sections [{
         type=MovieSceneParameterSection
         end_frame=9
         start_frame=0
         channels [ArrowScale {type=MovieSceneScriptingFloatChannel, keys [[0,1],[4,0],[8,1]]}]
        }]
      }]
    }
   ]
   display_rate {numerator=30}
  }
 }
 OnRiding {
  binding_count=2
  length_seconds=0.3
  object=&self:OnRiding
  sequence {
   bindings [
    Image_Reticle_Fuel {
     id=#3
     possessed_class=Image
     tracks [{
       type=MovieSceneWidgetMaterialTrack
       sections [{
         type=MovieSceneParameterSection
         end_frame=9
         start_frame=0
         channels [MasterOpacity {type=MovieSceneScriptingFloatChannel, keys [[3,0],[8,0.85],[9,0.85]]}]
        }]
      }]
    }
    Image_Reticle_Bottom_Frame {
     id=#4
     possessed_class=Image
     tracks [{
       type=MovieSceneWidgetMaterialTrack
       sections [{
         type=MovieSceneParameterSection
         end_frame=9
         start_frame=0
         channels [
          MasterOpacity {type=MovieSceneScriptingFloatChannel, keys [[3,0],[8,1],[9,1]]}
          ArrowOffset {type=MovieSceneScriptingFloatChannel, keys [[0,0],[8,1],[9,1]]}
         ]
        }]
      }]
    }
   ]
   display_rate {numerator=30}
  }
 }
 OnCooldown {
  binding_count=1
  length_seconds=0.1667
  object=&self:OnCooldown
  sequence {
   bindings [Image_Reticle_Bottom_Frame {
     id=#5
     possessed_class=Image
     tracks [{
       type=MovieSceneWidgetMaterialTrack
       sections [{
         type=MovieSceneParameterSection
         end_frame=5
         start_frame=0
         channels [MasterOpacity {type=MovieSceneScriptingFloatChannel, keys [[0,1],[4,0.3]]}]
        }]
      }]
    }]
   display_rate {numerator=30}
  }
 }
]
graphs [EventGraph {nodes [
   Event Construct()
   n1 = GetEquippedWeapon!(self:Fort Weapon Reticle Extension Widget Base Object Reference) -> (ReturnValue:Fort Weapon Object Reference)
   > Is Valid(InputObject=@n1)
   > n2 = Cast To FortWeaponRanged!(execute, Object=@n1) -> (then, CastFailed, AsFort Weapon Ranged:Fort Weapon Ranged Object Reference) {is=title K2Node_DynamicCast}
   > Set Weapon_RocketDrill(Weapon_RocketDrill:Fort Weapon Ranged Object Reference=@n2) -> (Output_Get:Fort Weapon Ranged Object Reference)
   > n3 = GetDynamicMaterial(self=Get Image_Reticle_Bottom_Frame() -> (Image_Reticle_Bottom_Frame:Image Object Reference))
   > n4 = Set M_Riding_Reticle(M_Riding_Reticle:Material Instance Dynamic Object Reference=@n3) -> (Output_Get:Material Instance Dynamic Object Reference)
   > SetScalarParameterValue(self=@n4, ParameterName=MasterOpacity, Value=0)
   > n5 = GetDynamicMaterial(self=Get Image_Reticle_Fuel() -> (Image_Reticle_Fuel:Image Object Reference))
   > n6 = Set M_Fuel_Reticle(M_Fuel_Reticle:Material Instance Dynamic Object Reference=@n5) -> (Output_Get:Material Instance Dynamic Object Reference)
   > SetScalarParameterValue(self=@n6, ParameterName=MasterOpacity, Value=0)
   n7 = GetOwningPlayerPawn!(self:User Widget Object Reference) -> (ReturnValue:Pawn Object Reference)
   > n8 = WaitGameplayTagAddToActor!(
    execute:Exec
    TargetActor:Actor Object Reference=@n7
    Tag:Gameplay Tag Structure=(TagName="Gameplay.Action.Player.RocketDrill.Active")
    OnlyTriggerOnce:bool=false
   ) -> (then:Exec, AsyncTaskProxy:Ability Async Wait Gameplay Tag Added Object Reference, Added:Exec) {is=title K2Node_AsyncAction}
   > n9 = WaitGameplayTagRemoveFromActor!(
    execute:Exec
    TargetActor:Actor Object Reference=@n7
    Tag:Gameplay Tag Structure=(TagName="Gameplay.Action.Player.RocketDrill.Active")
    OnlyTriggerOnce:bool=false
   ) -> (then:Exec, AsyncTaskProxy:Ability Async Wait Gameplay Tag Removed Object Reference, Removed:Exec) {is=title K2Node_AsyncAction}
   Branch(execute=@n9.Removed, Condition=Get DrillActive?() -> (DrillActive?:bool))
   > PlayAnimation(InAnimation=Get OnRiding() -> (OnRiding:Widget Animation Object Reference), PlayMode=Reverse)
   > Set DrillActive?(DrillActive?:bool=false) -> (Output_Get:bool)
   PlayAnimation(execute=@n8.Added, InAnimation=Get OnRiding())
   > Set DrillActive?(DrillActive?=true)
   UpdateAmmo!() -> (OutputDelegate, then, LocalCount:int, LocalRemaining:int) {is=title K2Node_CustomEvent}
   Event PreConstruct()
   Event Tick()
  ]}]
variables [
 DrillActive? {type=boolean}
 Image Reticle Fuel {type=Image, default=null}
 M_Fuel_Reticle {type=MaterialInstanceDynamic}
 M_Riding_Reticle {type=MaterialInstanceDynamic}
 Weapon_RocketDrill {type=FortWeaponRanged, default=null}
]
widgets {
 info {
  inheritedWidgetCount=1
  parentClass=/Script/FortniteUI.FortWeaponReticleExtensionWidgetBase
  rootWidgetClass=&Overlay
  widgetCount=6
 }
 widgets [
  Overlay_48 {widgetClassPath=&Overlay, widget=Overlay_48}
  Image_ReticleCenter {
   widgetClassPath=&Image
   bIsVariable=true
   parent=Overlay_48
   widget=Image_ReticleCenter
   properties {brush {
     bIsDynamicallyLoaded=false
     drawAs=Image
     imageType=FullColor
     mirroring=NoMirror
     resourceName=None
     resourceObject=&/M_UI_RocketDrill_Riding_Reticle
     tiling=NoTile
     imageSize=<48 48>
     margin=<ltrb 0 0 0 0>
     outlineSettings {
      bUseBrushTransparency=false
      roundingType=HalfHeightRadius
      width=0
      color {colorUseRule=UseColor_Specified, specifiedColor=<rgba 0 0 0 0>}
      cornerRadii=<0 0 0 0>
     }
     tintColor {colorUseRule=UseColor_Specified, specifiedColor=<rgba 1 1 1 1>}
     uVRegion {bIsValid=false, max=<0 0>, min=<0 0>}
    }}
   slot {class=OverlaySlot, properties {horizontalAlignment=HAlign_Center, verticalAlignment=VAlign_Center}}
  }
  Image_Reticle_Bottom_Frame {
   widgetClassPath=&Image
   bIsVariable=true
   parent=Overlay_48
   widget=Image_Reticle_Bottom_Frame
   properties {brush {
     bIsDynamicallyLoaded=false
     drawAs=Image
     imageType=FullColor
     mirroring=NoMirror
     resourceName=None
     resourceObject=&/M_UI_RocketDrill_Ground_Reticle
     tiling=NoTile
     imageSize=<100 100>
     margin=<ltrb 0 0 0 0>
     outlineSettings {
      bUseBrushTransparency=false
      roundingType=HalfHeightRadius
      width=0
      color {colorUseRule=UseColor_Specified, specifiedColor=<rgba 0 0 0 0>}
      cornerRadii=<0 0 0 0>
     }
     tintColor {colorUseRule=UseColor_Specified, specifiedColor=<rgba 1 1 1 1>}
     uVRegion {bIsValid=false, max=<0 0>, min=<0 0>}
    }}
   slot {class=OverlaySlot, properties {horizontalAlignment=HAlign_Center, verticalAlignment=VAlign_Center}}
  }
  Image_Reticle_Outer_Frame {
   widgetClassPath=&Image
   bIsVariable=true
   parent=Overlay_48
   widget=Image_Reticle_Outer_Frame
   properties {
    brush {
     bIsDynamicallyLoaded=false
     drawAs=Image
     imageType=FullColor
     mirroring=NoMirror
     resourceName=None
     resourceObject=&M_UI_RocketDrill_Riding_Reticle
     tiling=NoTile
     imageSize=<48 48>
     margin=<ltrb 0 0 0 0>
     outlineSettings {
      bUseBrushTransparency=false
      roundingType=HalfHeightRadius
      width=0
      color {colorUseRule=UseColor_Specified, specifiedColor=<rgba 0 0 0 0>}
      cornerRadii=<0 0 0 0>
     }
     tintColor {colorUseRule=UseColor_Specified, specifiedColor=<rgba 1 1 1 1>}
     uVRegion {bIsValid=false, max=<0 0>, min=<0 0>}
    }
    colorAndOpacity=<rgba 1 1 1 0>
    renderTransform {angle=0, scale=<1 1>, shear=<0 0>, translation=<0 25>}
   }
   slot {class=OverlaySlot, properties {horizontalAlignment=HAlign_Center, verticalAlignment=VAlign_Center}}
  }
  Image_Reticle_Fuel {
   widgetClassPath=&Image
   bIsVariable=true
   parent=Overlay_48
   widget=Image_Reticle_Fuel
   properties {
    visibility=Hidden
    brush {
     bIsDynamicallyLoaded=false
     drawAs=Image
     imageType=FullColor
     mirroring=NoMirror
     resourceName=None
     resourceObject=&/M_UI_RocketDrill_Fuel
     tiling=NoTile
     imageSize=<256 256>
     margin=<ltrb 0 0 0 0>
     outlineSettings {
      bUseBrushTransparency=false
      roundingType=HalfHeightRadius
      width=0
      color {colorUseRule=UseColor_Specified, specifiedColor=<rgba 0 0 0 0>}
      cornerRadii=<0 0 0 0>
     }
     tintColor {colorUseRule=UseColor_Specified, specifiedColor=<rgba 1 1 1 1>}
     uVRegion {bIsValid=false, max=<0 0>, min=<0 0>}
    }
    renderTransform {angle=0, scale=<1 1>, shear=<0 0>, translation=<75 25>}
   }
   slot {class=OverlaySlot, properties {horizontalAlignment=HAlign_Center, verticalAlignment=VAlign_Center}}
  }
  TutorialHighlight {widgetClassPath=&TutorialHighlight}
 ]
}
